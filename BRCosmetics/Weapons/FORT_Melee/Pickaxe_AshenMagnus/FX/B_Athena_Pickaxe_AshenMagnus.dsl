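; Blueprint for the Ashen Magnus cosmetic pickaxe that feeds the selected style variant into its swing and impact effects.
; On equip, it resolves the instigator pawn's active Particle variant channel (Particle1 or Particle2) and stores it as VariantIndex.
; Left and right swings send the montage section index and variant to the swing VFX, using section -1 when the cast fails.
; Impacts set the variant on the spawned effect, except in frontend display.
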(event FortWeapon|EventOnEquippedAndAttached
  (|CallParentFunction)
  (bind _asfort_player_pawn (Utilities|Casting|CastToFortPlayerPawn (Variables|Actor|GetInstigator)))
  (bind _index (CosmeticLoadout|GetVariantChannelActiveIndexforSlotActor "(TagName=\"LoadoutSlot.Pickaxe\")" _asfort_player_pawn "(TagName=\"Cosmetics.Variant.Channel.Particle\")" (Utilities|Array|MakeArray "(TagName=\"Cosmetics.Variant.Property.Particle1\")" "(TagName=\"Cosmetics.Variant.Property.Particle2\")")))
  (Variables|Default|SetVariantIndex _index))

(event EventSwingLeft
  (bind _returnvalue (CallFunction|GetSwingVFXComponent))
  (|CallParentFunction)
  (bind _section_index (Utilities|SetupSwingMontageSectionIndex (Variables|Actor|GetInstigator))
    (:Out
      (Effects|Components|ParticleSystem|SetIntParameter _returnvalue "SwingSectionIndex" _section_index)
      (Effects|Components|ParticleSystem|SetFloatParameter _returnvalue "Variant" (Math|Conversions|ToFloat(Integer) (Variables|Default|GetVariantIndex))))
    (:Cast%20Failed%20Out
      (Effects|Components|ParticleSystem|SetIntParameter _returnvalue "SwingSectionIndex" -1))))

(event EventSwingRight
  (bind _returnvalue (CallFunction|GetSwingVFXComponent))
  (|CallParentFunction)
  (bind _section_index (Utilities|SetupSwingMontageSectionIndex (Variables|Actor|GetInstigator))
    (:Out
      (Effects|Components|ParticleSystem|SetIntParameter _returnvalue "SwingSectionIndex" _section_index)
      (Effects|Components|ParticleSystem|SetFloatParameter _returnvalue "Variant" (Math|Conversions|ToFloat(Integer) (Variables|Default|GetVariantIndex))))
    (:Cast%20Failed%20Out
      (Effects|Components|ParticleSystem|SetIntParameter _returnvalue "SwingSectionIndex" -1))))

(event FortWeapon|EventOnPlayImpactFX
  (params (HitResult :type Hit%20Result%20Structure%20%28by%20ref%29) (ImpactPhysicalSurface :type EPhysicalSurface%20Enum :default "SurfaceType_Default") (SpawnedPSC :type FXSystem%20Component%20Object%20Reference))
  (|CallParentFunction HitResult ImpactPhysicalSurface SpawnedPSC)
  (if (not (Reactive|Frontend|IsCosmeticObjectDisplayedinFrontendContext 0))
    (Effects|Components|ParticleSystem|SetFloatParameter SpawnedPSC "Variant" (Math|Conversions|ToFloat(Integer) (Variables|Default|GetVariantIndex)))))

(fn ConstructionScript
  (|CallParentFunction))
